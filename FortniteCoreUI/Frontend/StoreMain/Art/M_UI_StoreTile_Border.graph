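# Translucent UI material that draws a soft-edged border for store tiles, with an animatable fill color.
# Four rotated UV masks form the edges, each with angle, offset and softness parameters, summed and scaled by OverallAlpha.
# AnimateAngle blends the edges from axis-aligned to slightly tilted.
# AnimateFill slides a vertical gradient that blends white into FillColor and boosts alpha only within the border mask.
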Material /FortniteCoreUI/Frontend/StoreMain/Art/M_UI_StoreTile_Border
graph {
 properties {
  bUsedWithStaticMesh=false
  blendMode=BLEND_Translucent
  materialDomain=MD_UI
  pixelDepthOffsetMode=PDOM_Legacy
  usedShadingModels=Default Lit
 }
 nodes [
  n1 = ComponentMask(None=TextureCoordinate()) {g=true}
  n2 = LinearInterpolate(A=Constant() {r=-1}, B=Constant() {r=1}, Alpha=ScalarParameter() {defaultValue=1, parameterName=AnimateFill})
  n3 = Clamp(None=Multiply(A=Add(A=@n1, B=@n2)))
  n4 = Constant() {r=-0.5}
  n5 = ScalarParameter() {defaultValue=-0.505, parameterName=Right_Angle}
  n6 = ScalarParameter() {defaultValue=1, parameterName=AnimateAngle}
  n7 = Add(
   A=Rotator(Coordinate=TextureCoordinate(), Time=LinearInterpolate(A=@n4, B=@n5, Alpha=@n6)) {speed=6.2832}
   B=ScalarParameter() {defaultValue=-0.05, parameterName=Right_Offset}
  )
  n8 = OneMinus(None=Multiply(A=ComponentMask(None=@n7) {r=true}, B=ScalarParameter() {defaultValue=128, parameterName=Right_Softness}))
  n9 = Constant()
  n10 = ScalarParameter() {defaultValue=-0.005, parameterName=Left_Angle}
  n11 = Add(
   A=Rotator(Coordinate=TextureCoordinate(), Time=LinearInterpolate(A=@n9, B=@n10, Alpha=@n6)) {speed=6.2832}
   B=ScalarParameter() {defaultValue=-0.05, parameterName=Left_Offset}
  )
  n12 = OneMinus(None=Multiply(A=ComponentMask(None=@n11) {r=true}, B=ScalarParameter() {defaultValue=128, parameterName=Left_Softness}))
  n13 = Clamp(None=Add(A=Clamp(None=@n8), B=Clamp(None=@n12)))
  n14 = Constant() {r=0.75}
  n15 = ScalarParameter() {defaultValue=0.755, parameterName=T_Angle}
  n16 = Add(
   A=Rotator(Coordinate=TextureCoordinate(), Time=LinearInterpolate(A=@n14, B=@n15, Alpha=@n6)) {speed=6.2832}
   B=ScalarParameter() {defaultValue=-0.05, parameterName=T_Offset}
  )
  n17 = OneMinus(None=Multiply(A=ComponentMask(None=@n16) {r=true}, B=ScalarParameter() {defaultValue=128, parameterName=T_Softness}))
  n18 = Constant() {r=0.25}
  n19 = ScalarParameter() {defaultValue=0.255, parameterName=B_Angle}
  n20 = Add(
   A=Rotator(Coordinate=TextureCoordinate(), Time=LinearInterpolate(A=@n18, B=@n19, Alpha=@n6)) {speed=6.2832}
   B=ScalarParameter() {defaultValue=-0.05, parameterName=B_Offset}
  )
  n21 = OneMinus(None=Multiply(A=ComponentMask(None=@n20) {r=true}, B=ScalarParameter() {defaultValue=128, parameterName=B_Softness}))
  n22 = Clamp(None=Add(A=Clamp(None=@n17), B=Clamp(None=@n21)))
  n23 = Multiply(A=Clamp(None=Add(A=@n13, B=@n22)), B=ScalarParameter() {defaultValue=1, parameterName=OverallAlpha})
  Clamp(None=Add(A=DotProduct(A=@n3, B=@n23), B=@n23))
  Clamp(None=LinearInterpolate(A=Constant3Vector() {constant=<rgba 1 1 1 0>}, B=VectorParameter() {parameterName=FillColor, defaultValue=<rgba 1 1 1 1>}.RGB, Alpha=@n3))
 ]
}
