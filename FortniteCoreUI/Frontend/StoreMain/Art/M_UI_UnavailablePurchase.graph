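# Translucent UI material in the store front end that fades an offer image's alpha, named for the unavailable-purchase display.
# It samples the OfferTexture parameter, which defaults to a Black Monday outfit image, and takes its alpha.
# A vertical gradient from the 9-slice UV, shifted by Displacement, inverted, scaled by the gradient size parameter and saturated, multiplies that alpha, fading it toward the bottom.
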Material /FortniteCoreUI/Frontend/StoreMain/Art/M_UI_UnavailablePurchase
graph {
 properties {
  bUsedWithStaticMesh=false
  blendMode=BLEND_Translucent
  materialDomain=MD_UI
  pixelDepthOffsetMode=PDOM_Legacy
  usedShadingModels=Default Lit
 }
 nodes [
  n1 = TextureObjectParameter() {
   parameterName=OfferTexture
   texture=/FortniteCoreUI/UI/Foundation/Textures/BattleRoyale/FeaturedItems/Outfit/T-AthenaSoldiers-CID-Athena-Commando-BlackMonday_Details
  }
  n2 = TextureSample(Tex=@n1) {texture=&T-AthenaSoldiers-CID-Athena-Commando-BlackMonday_Details}
  n3 = MaterialFunctionCall() {materialFunction=&GetUserInterfaceUV}
  n4 = ScalarParameter() {defaultValue=0.15, parameterName=Displacement}
  n5 = OneMinus(None=ComponentMask(None=Add(A=@n3.9-Slice UV, B=@n4) {constB=0.25}) {g=true})
  n6 = ScalarParameter() {defaultValue=2, parameterName=Gradient SIze}
  Multiply(A=@n2.A, B=Saturate(None=Multiply(A=@n5, B=@n6)))
 ]
}
